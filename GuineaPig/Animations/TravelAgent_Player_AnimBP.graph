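# Player animation Blueprint for the TravelAgent launcher, providing item override layers that animate weapon charging and max-charge firing.
# ItemUpperBody, the only layer with animation logic, applies an additive charge intro plus a ChargeAlpha-weighted loop while charging.
# Each update detects the fire montage by name and latches max charge until curve UserCurve01 clears it.
# When both flags are set, a resettable DoOnce jumps the montage to its Charge section.
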AnimBlueprint /GuineaPig/Animations/TravelAgent_Player_AnimBP
target_skeleton=&Fortnite_M_Avg_Player_Skeleton
graphs [
 AnimGraph {nodes [Output Pose()]}
 EventGraph {nodes [
   Event BlueprintUpdateAnimation()
   > n1 = Sequence!(execute) -> (then_0, then_1, then_2:Exec)
   > n2 = Cast To FortPlayerPawn(Object=TryGetFortPawn())
   n3 = GetAnimInstance(self=Get Mesh(self=@n2))
   n4 = GetCurrentActiveMontage!(self:Anim Instance Object Reference=@n3) -> (ReturnValue:Anim Montage Object Reference)
   > Set IsFireMontagePlaying(IsFireMontagePlaying:bool="Equal Exactly (String)"(A="To String (Object)"!(InObj:Object Reference=@n4) -> (ReturnValue:String), B=Launcher_TravelAgent_Fire_Montage)) -> (Output_Get:bool)
   n5 = Branch(execute=@n1.then_2, Condition=Get bWeaponIsAtMaxCharge() -> (bWeaponIsAtMaxCharge:bool))
   > Set IsMaxCharge(IsMaxCharge:bool=true) -> (Output_Get:bool)
   Branch(execute=@n5.else, Condition="float > float"(A=GetCurveValue(CurveName=UserCurve01), B=0.5))
   > Set IsMaxCharge(IsMaxCharge=false)
   n6 = Branch(
    execute=@n1.then_1
    Condition=AND Boolean(A=Get IsFireMontagePlaying() -> (IsFireMontagePlaying:bool), B=Get IsMaxCharge() -> (IsMaxCharge:bool))
   )
   > Do Once(Reset=@n6.else)
   > Montage_JumpToSection!(execute, self:Anim Instance Object Reference=@n3, SectionName:Name=Charge, Montage:Anim Montage Object Reference=@n4) -> (then)
  ]}
 ItemAimOffset {nodes [
   Input Pose!() -> (Pose, AimOffsetAlpha:double, Yaw:double, Pitch:double) {name=InPose_AimOffset}
   Input Pose!() -> (Pose) {name=InPose_PreAimOffsets}
   Output Pose() {layerGroup=Item Override}
  ]}
 ItemFallAdditive {nodes [
   Input Pose() {name=InPoseItemFallAdditive}
   Output Pose() {layerGroup=Item Override}
  ]}
 ItemFallLandAdditive {nodes [
   Input Pose() {name=InPoseItemFallLandAdditive}
   Output Pose() {layerGroup=Item Override}
  ]}
 ItemFallLevel2Override {nodes [
   Input Pose() {name=InPose_FallLevel2}
   Output Pose() {layerGroup=Item Override}
  ]}
 ItemFinalPoseOverride {nodes [
   Input Pose() {name=InFinalPose}
   Output Pose() {layerGroup=Item Override}
  ]}
 ItemFullBodyAimOffsets {nodes [
   Input Pose!() -> (Pose, FullBody_AimOffsetAlpha:double, Aim_Yaw:double, Aim_Pitch:double) {name=InPoseFullBodyAimOffsets}
   Input Pose!() -> (Pose) {name=InPosePreFullBodyAimOffsets}
   Output Pose() {layerGroup=Item Override}
  ]}
 ItemFullBodyOverride {nodes [
   Input Pose() {name=InPoseFullBody}
   Output Pose() {layerGroup=Item Override}
  ]}
 ItemIdleOverride {nodes [
   Input Pose() {name=InPose_Idle}
   Output Pose() {layerGroup=Item Override}
  ]}
 ItemJetPackJumpAdditive {nodes [Output Pose() {layerGroup=Item Override}]}
 ItemJetPackStartAdditive {nodes [Output Pose() {layerGroup=Item Override}]}
 ItemJumpLoopAdditive {nodes [
   Input Pose() {name=InPoseItemJumpLoopAdditive}
   Output Pose() {layerGroup=Item Override}
  ]}
 ItemJumpOverride {nodes [
   Input Pose() {name=InPoseJump}
   Output Pose() {layerGroup=Item Override}
  ]}
 ItemJumpUpAdditive {nodes [
   Input Pose() {name=InPoseItemJumpUpAdditive}
   Output Pose() {layerGroup=Item Override}
  ]}
 ItemPreAimOffsets {nodes [
   Input Pose() {name=InPosePreAimOffsets}
   Output Pose() {layerGroup=Item Override}
  ]}
 ItemPreIK {nodes [
   Input Pose() {name=InPose_PreIK}
   Output Pose() {layerGroup=Item Override}
  ]}
 ItemRightHandIK {nodes [
   Input Pose() {name=InPosePreRightHandIK}
   Input Pose() {name=InPosePostRightHandIK}
   Output Pose() {layerGroup=Item Override}
  ]}
 ItemSkeletalControl {nodes [
   Input Pose()
   Output Pose() {layerGroup=Item Override}
  ]}
 ItemSwimJumpFallLoop {nodes [Output Pose() {layerGroup=Item Override}]}
 ItemSwimJumpSurfaceEnd {nodes [Output Pose() {layerGroup=Item Override}]}
 ItemSwimJumpSurfaceLoop {nodes [Output Pose() {layerGroup=Item Override}]}
 ItemUpperBody {nodes [
   n1 = "Sequence Player (additive) 'Launcher_TravelAgent_Charge_Loop'"() {sequence=&/Launcher_TravelAgent_Charge_Loop}
   AnimGraphNode_StateMachine_0 = UB Charge SM() {is=title AnimGraphNode_StateMachine}
   n2 = Get ChargeAlpha() -> (ChargeAlpha:float)
   Output Pose(Result=Blend Poses by bool(
     BlendPose_0=Apply Additive(Base=@AnimGraphNode_StateMachine_0, Additive=@n1, Alpha=@n2) {alphaBoolBlend {blendOption=Linear}}
     BlendPose_1=Use cached pose 'UB Pose Cached'()
     bActiveValue=Get bWeaponIsCharging() -> (bWeaponIsCharging:bool)
    ) {blendType=Linear}) {layerGroup=Item Override}
   Save cached pose 'UB Pose Cached'(Pose=Input Pose!() -> (Pose, InputParam:Fort Anim Input Adjusted Aim Structure) {name=InPoseUpperBody}) {cacheName=UB Pose Cached}
  ]}
 UB Charge SM {
  entry=UB Charge Intro
  path=ItemUpperBody.AnimGraphNode_StateMachine_0.UB Charge SM
  ids=id
  nodes [AnimStateNode_0 = UB Charge Intro() {is=title AnimStateNode}]
 }
 UB Charge Intro {
  path=ItemUpperBody.AnimGraphNode_StateMachine_0.UB Charge SM.AnimStateNode_0.UB Charge Intro
  nodes [Output Animation Pose(Result=Apply Additive(
     Base=Use cached pose 'UB Pose Cached'()
     Additive="Sequence Player (additive) 'Launcher_TravelAgent_Charge_intro'"() {bLoopAnimation=false, sequence=&/Launcher_TravelAgent_Charge_intro}
    ) {alphaBoolBlend {blendOption=Linear}})]
 }
 ItemZipLineStartAdditive {nodes [Output Pose() {layerGroup=Item Override}]}
]
variables [
 IsFireMontagePlaying {type=boolean}
 IsMaxCharge {type=boolean}
]
